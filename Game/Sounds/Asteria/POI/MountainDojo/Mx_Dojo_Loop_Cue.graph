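# Looping positional music cue for the Mountain Dojo POI, playing a single dojo loop wave at reduced volume.
# One looping wave player feeds the output directly, using a small-speaker music attenuation preset borrowed from Neon City.
# A concurrency override allows one instance, and play-when-silent virtualization keeps the loop running while inaudible.
# It also sends to a music audio-analysis submix.
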SoundCue /Game/Sounds/Asteria/POI/MountainDojo/Mx_Dojo_Loop_Cue
root=WavePlayer_0
nodes [{id=WavePlayer_0, properties {bLooping=true, soundWaveAssetPtr=&/MX_S24_Dojo_POI_loop}}]
properties {
 attenuationSettings=/Game/Sounds/Asteria/POI/NeonCity/mx_NeonCity/DiegeticMusic/NeonCityPresets/Speakers_Cue/Speakers_POI_Sm/NeonCity_CUE_Music_SmSpeaker_Att
 bOverrideConcurrency=true
 maxDistance=4500
 virtualizationMode=PlayWhenSilent
 volumeMultiplier=0.19
 attenuationOverrides {falloffDistance=1000, nonFocusVolumeAttenuation=0, nonSpatializedRadiusStart=500, stereoSpread=1500, attenuationShapeExtents {x=300}}
 concurrencyOverrides {maxCount=1}
 soundSubmixSends [{minSendLevel=0, soundSubmix=&Fortnite_AudioAnalysis_Music}]
}
